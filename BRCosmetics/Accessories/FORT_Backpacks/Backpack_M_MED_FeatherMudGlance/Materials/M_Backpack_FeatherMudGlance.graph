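# Character-parent material for the FeatherMudGlance backpack that adds an animated flipbook display and a tinted glow to emissive.
# Frames from a 4x4 SubUV atlas, some fixed and others driven by sine time and Animation Phase parameters, blend through FX2 texture masks.
# A pixelated sphere mask sized by Reset radius switches between static and animated composites.
# KillGlow scales the Fresnel and masked glow, and the result is injected between PreFX and PostFX functions.
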Material /BRCosmetics/Accessories/FORT_Backpacks/Backpack_M_MED_FeatherMudGlance/Materials/M_Backpack_FeatherMudGlance
graph {
 properties {
  bUseMaterialAttributes=true
  bUsedWithMeshDeformer=true
  bUsedWithMorphTargets=true
  bUsedWithSkeletalMesh=true
  pixelDepthOffsetMode=PDOM_Legacy
  shadingModel=MSM_FromMaterialExpression
  usedShadingModels=Default Lit | Subsurface | Clear Coat | Subsurface Profile | Cloth
 }
 nodes [
  n1 = MaterialFunctionCall() {materialFunction=&MF_CharacterParent_2023_PreFX}
  n2 = GetMaterialAttributes(None=@n1.Result) {attributeGetTypes [#1, #2]}
  n3 = NamedRerouteUsage()
  n4 = ScalarParameter() {defaultValue=1, parameterName=Specular multiple}
  MaterialFunctionCall(
   Material=SetMaterialAttributes!(MaterialAttributes=@n2.MaterialAttributes, Emissive Color=Add(A=@n3, B=@n2.Emissive Color), Specular=Multiply(A=@n2.Specular, B=@n4)) {attributeSetTypes [#1, #2]}
   M=@n1.M
   Pre Skinned Local Position=@n1.M
   Distance Blend=@n1.M
   Fresnel=@n1.M
  ) {materialFunction=&MF_CharacterParent_2023_PostFX}
  n5 = Time()
  n6 = ConstantBiasScale(None=Sine(None=Multiply(A=@n5)))
  n7 = ScalarParameter() {parameterName=Animation Phase}
  Multiply(A=LinearInterpolate(Alpha=Saturate(None=@n6)) {constA=0.62, constB=0.44}, B=@n7)
  n8 = Constant() {r=0.936}
  n9 = Constant() {r=4}
  n10 = MaterialFunctionCall(UVs=TextureCoordinate(), Texture Scale=Constant() {r=0.45}) {materialFunction=&ScaleUVsByCenter}
  n11 = Constant() {r=0.16}
  n12 = Constant() {r=-0.008}
  n13 = TextureSample(UVs=Add(
    A=MaterialFunctionCall("Animation  Phase (0-1)"=@n8, Number of Rows=@n9, Number of Columns=@n9, UVs=@n10.UVs) {materialFunction=&FlipBook}.UVs
    B=AppendVector(A=@n11, B=@n12)
   )) {
   samplerSource=SSM_Wrap_WorldGroupSettings
   texture=/BRCosmetics/Accessories/FORT_Backpacks/Backpack_M_MED_FeatherMudGlance/Textures/T_FeatherMudGlance_SubUV
  }
  n14 = Constant() {r=4}
  n15 = TextureCoordinate()
  n16 = Constant() {r=0.45}
  n17 = Add(
   A=MaterialFunctionCall(
    "Animation  Phase (0-1)"=Constant() {r=0.87}
    Number of Rows=@n14
    Number of Columns=@n14
    UVs=MaterialFunctionCall(UVs=@n15, Texture Scale=@n16) {materialFunction=&ScaleUVsByCenter}.UVs
   ) {materialFunction=&FlipBook}.UVs
   B=AppendVector(A=Constant() {r=0.168}, B=Constant() {r=-0.014})
  )
  n18 = TextureSample(UVs=@n17) {
   samplerSource=SSM_Wrap_WorldGroupSettings
   texture=&T_FeatherMudGlance_SubUV
  }
  n19 = TextureSample(UVs=@n17) {
   samplerSource=SSM_Wrap_WorldGroupSettings
   texture=/BRCosmetics/Accessories/FORT_Backpacks/Backpack_M_MED_FeatherMudGlance/Textures/T_FeatherMudGlance_Backpack_FX2
  }
  n20 = Constant() {r=0.9841}
  n21 = Constant() {r=4}
  n22 = MaterialFunctionCall(UVs=TextureCoordinate(), Texture Scale=Constant() {r=0.45}) {materialFunction=&ScaleUVsByCenter}
  n23 = Constant() {r=0.16}
  n24 = Constant() {r=-0.008}
  n25 = TextureSample(UVs=Add(
    A=MaterialFunctionCall("Animation  Phase (0-1)"=@n20, Number of Rows=@n21, Number of Columns=@n21, UVs=@n22.UVs) {materialFunction=&FlipBook}.UVs
    B=AppendVector(A=@n23, B=@n24)
   )) {
   samplerSource=SSM_Wrap_WorldGroupSettings
   texture=&T_FeatherMudGlance_SubUV
  }
  n26 = ScalarParameter() {parameterName=Animation Phase2}
  n27 = Constant()
  n28 = Constant() {r=1}
  n29 = LinearInterpolate(A=@n13.RGB, Alpha=Multiply(A=@n25.R, B=If(A=@n26, "A > B"=@n27, "A == B"=@n27, "A < B"=@n28) {constB=0.814}))
  n30 = Time()
  n31 = ConstantBiasScale(None=Sine(None=Multiply(A=@n30) {constB=0.18}))
  n32 = Max(A=LinearInterpolate(Alpha=Saturate(None=@n31)) {constB=0.432}, B=ScalarParameter() {parameterName=Animation Phase2})
  n33 = Constant() {r=4}
  n34 = TextureCoordinate()
  n35 = Constant() {r=0.45}
  n36 = Add(
   A=MaterialFunctionCall(
    "Animation  Phase (0-1)"=Max(A=@n7, B=@n32)
    Number of Rows=@n33
    Number of Columns=@n33
    UVs=MaterialFunctionCall(UVs=@n34, Texture Scale=@n35) {materialFunction=&ScaleUVsByCenter}.UVs
   ) {materialFunction=&FlipBook}.UVs
   B=AppendVector(A=Constant() {r=0.168}, B=Constant() {r=-0.014})
  )
  n37 = TextureSample(UVs=@n36) {
   samplerSource=SSM_Wrap_WorldGroupSettings
   texture=&T_FeatherMudGlance_SubUV
  }
  n38 = TextureSample(UVs=@n36) {
   samplerSource=SSM_Wrap_WorldGroupSettings
   texture=&T_FeatherMudGlance_Backpack_FX2
  }
  n39 = TextureCoordinate()
  n40 = Divide(A=Floor(None=Multiply(A=@n39) {constB=90})) {constB=90}
  n41 = AppendVector(A=Constant() {r=0.214}, B=Constant() {r=0.504})
  n42 = ScalarParameter() {defaultValue=0.5, parameterName=Reset radius}
  n43 = Divide(A=Floor(None=Multiply(A=@n42) {constB=30})) {constB=30}
  n44 = Constant() {r=1}
  NamedRerouteDeclaration(None=LinearInterpolate(
    A=LinearInterpolate(A=@n13.RGB, B=@n18.RGB, Alpha=@n19.G)
    B=LinearInterpolate(A=@n29, B=@n37.RGB, Alpha=@n38.G)
    Alpha=SphereMask(A=@n40, B=@n41, Radius=@n43, Hardness=@n44)
   )) {name=Reactive, nodeColor=<rgba 1 0.8672 0 1>}
  n45 = VertexNormalWS()
  n46 = TextureSample() {
   samplerSource=SSM_Wrap_WorldGroupSettings
   texture=&T_FeatherMudGlance_Backpack_FX2
  }
  n47 = Multiply(
   A=Multiply(A=VectorParameter() {parameterName=EmissiveColor, defaultValue=<rgba 4 0.4157 3.6692 1>}.RGB, B=OneMinus(None=@n46.R))
   B=ScalarParameter() {parameterName=KillGlow}
  )
  n48 = Multiply(A=Power(Base=Fresnel(Normal=@n45) {exponent=2}, Exp=Constant() {r=2}) {constExponent=1.3}, B=@n47)
  n49 = Multiply(A=@n47, B=Multiply(A=@n46.B) {constB=0.8})
  n50 = NamedRerouteUsage()
  NamedRerouteDeclaration(None=Add(A=Add(A=@n48, B=@n49), B=LinearInterpolate(B=@n50, Alpha=@n46.R))) {name=Emissive out, nodeColor=<rgba 0.5469 0 1 1>}
 ]
}
